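# Mono, attenuated MetaSound source for ScubaBondOak gameplay music, apparently the third-person (3P) loop variant, on the Fortnite gameplay music sound class.
# On play, a random pick from five music waves feeds a wave player, whose completion fires OnFinished.
# Despite the Loop name, the player does not enable looping, and the Spark Layers boost-wave input is unconnected.
# A concurrency override limits instances per owner.
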MetaSoundSource /ScubaBondOakGameplay/Assets/Audio/Music/MSS_ScubaBondOak_Mx_Loop_3P
inputs [
 Music {type=WaveAsset:Array}
 Spark Layers {type=WaveAsset:Array}
 UE.Source.OnPlay {type=Trigger}
]
interfaces [UE.OutputFormat.Mono:1.0, UE.Source.OneShot:1.0, UE.Source:1.1]
outputs [
 "UE.OutputFormat.Mono.Audio:0" {type=Audio}
 UE.Source.OneShot.OnFinished {type=Trigger}
]
pages [Default {
  input_defaults {
   UE.Source.OnPlay=null
   Music [
    /ScubaBondOakGameplay/Assets/Audio/Music/wav/ScubaBond_Oak_IP_mx_3P_lp_01
    /ScubaBondOakGameplay/Assets/Audio/Music/wav/ScubaBond_Oak_IP_mx_3P_lp_02
    /ScubaBondOakGameplay/Assets/Audio/Music/wav/ScubaBond_Oak_IP_mx_3P_lp_03
    /ScubaBondOakGameplay/Assets/Audio/Music/wav/ScubaBond_Oak_IP_mx_3P_lp_04
    /ScubaBondOakGameplay/Assets/Audio/Music/wav/ScubaBond_Oak_IP_mx_3P_lp_05
   ]
   Spark Layers [
    /ScubaBondOakGameplay/Assets/Audio/SFX/wav/ScubaBondOak_Sky_BoostLayer_01
    /ScubaBondOakGameplay/Assets/Audio/SFX/wav/ScubaBondOak_Sky_BoostLayer_02
    /ScubaBondOakGameplay/Assets/Audio/SFX/wav/ScubaBondOak_Sky_BoostLayer_03
    /ScubaBondOakGameplay/Assets/Audio/SFX/wav/ScubaBondOak_Sky_BoostLayer_04
   ]
  }
  nodes [
   ArrayRandomGet_0 = Array.Random Get:WaveAsset:Array(In Array=^Music, Next=^UE.Source.OnPlay)
   WavePlayer_0 = UE.Wave Player:Mono(Play=@ArrayRandomGet_0.On Next, Wave Asset=@ArrayRandomGet_0.Value)
  ]
  links [{dst="@UE.Source.OneShot.OnFinished", src=WavePlayer_0.On Finished}, {dst="@UE.OutputFormat.Mono.Audio:0", src=WavePlayer_0.Out Mono}]
 }]
properties {
 assetImportData=&self:AssetImportData
 attenuationSettings=/ScubaBondOakGameplay/Assets/Audio/SFX/Att_ScubaBondOak_Music
 bOverrideConcurrency=true
 maxDistance=71500
 sampleRate=48000
 soundClassObject=/Game/Sounds/FortSoundClassesAndModes/Fort_Music_Gameplay
 soundSubmixObject=&Fortnite_SFX
 volume=6.4
 assetUserData [&self:AudioAssetUserData_0]
 concurrencyOverrides {bLimitToOwner=true, maxCount=2, retriggerTime=0.2}
}
